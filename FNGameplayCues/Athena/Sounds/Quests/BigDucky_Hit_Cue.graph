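# Sound cue for impact hits on the big rubber duck, stored with the Athena quest gameplay cue sounds.
# A Random node picks one of five equally weighted impact waves and feeds it through a default-setting Modulator for slight variation.
# Playback uses the impacts sound class, gadget-activate attenuation and a slightly reduced volume.
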SoundCue /FNGameplayCues/Athena/Sounds/Quests/BigDucky_Hit_Cue
root=Modulator_1
nodes [
 {id=Modulator_1, children [Random_0]}
 {id=Random_0, children [WavePlayer_0,WavePlayer_3,WavePlayer_4,WavePlayer_5,WavePlayer_6], properties {weights [1,1,1,1,1]}}
 {id=WavePlayer_0, properties {soundWaveAssetPtr=&/BigRubberDuck_Impacts_05}}
 {id=WavePlayer_3, properties {soundWaveAssetPtr=&/BigRubberDuck_Impacts_01}}
 {id=WavePlayer_4, properties {soundWaveAssetPtr=&/BigRubberDuck_Impacts_02}}
 {id=WavePlayer_5, properties {soundWaveAssetPtr=&/BigRubberDuck_Impacts_03}}
 {id=WavePlayer_6, properties {soundWaveAssetPtr=&/BigRubberDuck_Impacts_04}}
]
properties {
 attenuationSettings=&Fort_GadgetActivate_Att
 maxDistance=4000
 soundClassObject=&Fort_Impacts
 volumeMultiplier=0.8
}
